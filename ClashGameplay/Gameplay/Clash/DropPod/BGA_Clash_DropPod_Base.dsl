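; Base drop pod actor Blueprint for Clash gameplay, layering drop and landing handling onto its parent class.
; On landing it deactivates the drop Niagara effect, then calls the parent; the client landing event fades out the drop sound.
; BeginPlay, landing and server death print debug messages, while BeginDrop calls the parent and a cosmetic-event check with nothing shown after it.
; Overlap and tick events are disabled stubs, and OnRep_Open is empty.
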(event EventBeginPlay
  (|CallParentFunction)
  (Development|PrintString "Clash drop pod is alive"))

(event Collision|EventActorBeginOverlap
  (params (OtherActor :type Actor%20Object%20Reference))
  (|CallParentFunction OtherActor) ; This node is disabled and will not be called.\nDrag off pins to build functionality.
)

(event EventTick
  (params (DeltaSeconds :type Float%20%28single-precision%29))
  (|CallParentFunction DeltaSeconds) ; This node is disabled and will not be called.\nDrag off pins to build functionality.
)

(event EventOnDeathServer
  (params (Damage :type Float%20%28single-precision%29) (DamageTags :type Gameplay%20Tag%20Container%20Structure%20%28by%20ref%29) (Momentum :type Vector :default "0, 0, 0") (HitInfo :type Hit%20Result%20Structure%20%28by%20ref%29) (InstigatedBy :type Controller%20Object%20Reference) (DamageCauser :type Actor%20Object%20Reference) (EffectContext :type Gameplay%20Effect%20Context%20Handle%20Structure))
  (Development|PrintString "Clash drop pod destroyed"))

(event EventOnLandedClient
  (|CallParentFunction)
  (Audio|Components|Audio|FadeOut (Variables|Default|GetSound_Drop) 0.1 0.0 "Linear" :!class /Script/Engine.AudioComponent))

(fn BeginDrop
  (|CallParentFunction)
  (Networking|CanExecuteCosmeticEvents))

(fn OnLanded
  (Components|Activation|Deactivate (Variables|Default|GetNS_ClashGameplay_Drop))
  (Development|PrintString "Symbiote drop pod landed")
  (|CallParentFunction))

(fn OnRep_Open)

(fn ConstructionScript
  (|CallParentFunction))
