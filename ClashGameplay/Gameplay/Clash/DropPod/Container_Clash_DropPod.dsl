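; Clash drop pod container that gives loot from its loot tier group once, when searched or destroyed.
; Interacting marks it searched, then grants a picked item directly if the player holds under five items, playing a looted cue and dying, otherwise spawns loot.
; Death before searching spawns loot with a destroy cue; death effects hide the mesh.
; Helper functions play opening cue and smoke, and debug prints remain in the grant path.
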(event Interaction|EventBlueprintOnInteract
  (params (InteractingPawn :type Fort%20Pawn%20Object%20Reference) (InteractionBeingAttempted :type EInteractionBeingAttempted%20Enum :default "FirstInteraction"))
  (if (not (Variables|Default|GetSearched))
    (Variables|Default|SetSearched true)
    (bind _output_get (Variables|Default|SetLastInteractingPawn InteractingPawn))
    (bind _returnvalue (CallFunction|HasOpenInventorySlot _output_get))
    (if _returnvalue
      (CallFunction|GiveItemDirectly)
      (Ability|ExecuteGameplayCueWithParamsReplicated (Variables|Default|GetGC_Looted) (Ability|GameplayCue|MakeGameplayCueParameters :NormalizedMagnitude 0.0 :RawMagnitude 0.0 :Location (Transformation|GetWorldLocation (Variables|Default|GetFXPos)) :Normal "0,0,0" :Instigator self :GameplayEffectLevel 1 :AbilityLevel 1 :bReplicateLocationWhenUsingMinimalRepProxy false))
      (Components|Activation|Deactivate (Variables|Default|GetAthenaSpecialActor))
      (Gameplay|Die)
      (else
        (CallFunction|SpawnLoot)))))

(event Custom|GiveItemDirectly
  (Development|PrintString "Granting item")
  (bind _asfort_player_controller (Utilities|Casting|CastToFortPlayerController (Pawn|GetController (Variables|Default|GetLastInteractingPawn))))
  (bind _item_definition (CallFunction|GetItemtoGrant))
  (bind _asfort_world_item_definition (Utilities|Casting|CastToFortWorldItemDefinition _item_definition))
  (Development|PrintString (Utilities|GetDisplayName _asfort_world_item_definition))
  (Inventory|GiveItemtoInventoryOwner _asfort_player_controller _asfort_world_item_definition))

(event EventOnDeathServer
  (params (Damage :type Float%20%28single-precision%29) (DamageTags :type Gameplay%20Tag%20Container%20Structure%20%28by%20ref%29) (Momentum :type Vector :default "0, 0, 0") (HitInfo :type Hit%20Result%20Structure%20%28by%20ref%29) (InstigatedBy :type Controller%20Object%20Reference) (DamageCauser :type Actor%20Object%20Reference) (EffectContext :type Gameplay%20Effect%20Context%20Handle%20Structure))
  (if (not (Variables|Default|GetSearched))
    (CallFunction|SpawnLoot)
    (Ability|ExecuteGameplayCueWithParamsReplicated (Variables|Default|GetGC_Destroy) (Ability|GameplayCue|MakeGameplayCueParameters :NormalizedMagnitude 0.0 :RawMagnitude 0.0 :Location (Transformation|GetActorLocation) :Normal "0,0,0" :GameplayEffectLevel 1 :AbilityLevel 1 :bReplicateLocationWhenUsingMinimalRepProxy false))))

(event EventOnDeathPlayEffects
  (params (Damage :type Float%20%28single-precision%29) (DamageTags :type Gameplay%20Tag%20Container%20Structure%20%28by%20ref%29) (Momentum :type Vector :default "0, 0, 0") (HitInfo :type Hit%20Result%20Structure%20%28by%20ref%29) (InstigatedBy :type Fort%20Pawn%20Object%20Reference) (DamageCauser :type Actor%20Object%20Reference) (EffectContext :type Gameplay%20Effect%20Context%20Handle%20Structure))
  (Development|SetHiddeninGame (Variables|Default|GetStaticMesh) true true))

(event Custom|SpawnLoot
  (Class|AthenaFunctionLibrary|SpawnLootfromTable (Variables|Default|GetLootTierGroupName) 1 -1 (Transformation|GetActorLocation)))

(fn BlueprintCanInteract
  (params (InteractingPawn :type Fort%20Pawn%20Object%20Reference) (InteractionBeingAttempted :type EInteractionBeingAttempted%20Enum :default "FirstInteraction") (InteractionType :type TInteractionType%20Enum :default "IT_NoInteraction"))
  (return (not (Variables|Default|GetSearched))))

(fn BlueprintGetInteractionString
  (params (InteractingPawn :type Fort%20Pawn%20Object%20Reference) (InteractionBeingAttempted :type EInteractionBeingAttempted%20Enum :default "FirstInteraction"))
  (return (Variables|Default|GetInteractMessage)))

(fn BlueprintOnLocalInteract
  (params (InteractingPawn :type Fort%20Player%20Pawn%20Object%20Reference) (InteractionBeingAttempted :type EInteractionBeingAttempted%20Enum :default "FirstInteraction"))
  (return false))

(fn GetItemToGrant
  (bind _outloottodrop (Loot|PickLootDrops (Variables|Default|GetLootTierGroupName) 0 0 "(GameplayTags=)"))
  (Development|PrintString (Utilities|GetDisplayName (Class|FortItemFunctionLibrary|BreakFortItemEntry (Utilities|Array|Get(aref) _outloottodrop))))
  (return (Class|FortItemFunctionLibrary|BreakFortItemEntry (Utilities|Array|Get(aref) _outloottodrop))))

(fn HasOpenInventorySlot
  (params (Target :type Pawn%20Object%20Reference))
  (bind _asfort_player_controller (Utilities|Casting|CastToFortPlayerController (Pawn|GetController Target))
    (:then
      (bind _returnvalue (IndexSelectionModel|GetNumItems(Message) _asfort_player_controller))
      (Development|PrintString (Utilities|String|ToString(Integer) _returnvalue))
      (return (< _returnvalue 5)))
    (:CastFailed
      (return false))))

(fn OnOpenSequenceFinished
  (Variables|Default|SetFinishedOpening true)
  (Components|Activation|Deactivate (Variables|Default|GetInstancedSmokeFX)))

(fn PlayOpenFX
  (bind _returnvalue (Transformation|GetActorLocation))
  (GameplayCue|ExecuteGameplayCue%28Burst%29Local self (Variables|Default|GetGC_Open) (Ability|GameplayCue|MakeGameplayCueParameters :NormalizedMagnitude 0.0 :RawMagnitude 0.0 :Location _returnvalue :Normal "0,0,0" :Instigator self :GameplayEffectLevel 1 :AbilityLevel 1 :bReplicateLocationWhenUsingMinimalRepProxy false))
  (bind _returnvalue_2 (Effects|Components|ParticleSystem|SpawnAltFXAtLocation "GenericAnimNotify" 0 (Variables|Default|GetPS_InnerSmoke) _returnvalue))
  (Variables|Default|SetInstancedSmokeFX _returnvalue_2))

(fn ConstructionScript
  (|CallParentFunction))
